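# Material function for the Eternity backpack that layers an atmosphere rim and drifting clouds onto the input material attributes.
# The FX mask's blue channel drives a metals LUT and confines both the blue Fresnel rim and the clouds.
# Panning UV1 clouds, distorted by a time-animated flowmap, are broken up by noise, band-masked, then blended into color, normal and emissive.
# A final PanningEmissive function uses the mask's green channel as alpha.
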MaterialFunction /BRCosmetics/Accessories/FORT_Backpacks/Backpack_F_MED_Eternity/Materials/MF_F_MED_Eternity_Backpack
graph {nodes [
  n1 = TextureSample() {samplerType=SAMPLERTYPE_Masks, texture=/BRCosmetics/Accessories/FORT_Backpacks/Backpack_F_MED_Eternity/Textures/T_F_MED_Eternity_Backpack_FX}
  n2 = GetMaterialAttributes(None=MaterialFunctionCall(Mask=Multiply(A=@n1.B) {constB=0.666}, Materials=FunctionInput() {inputType=FunctionInput_MaterialAttributes}) {materialFunction=&MF_MetalsLUT}.Result) {attributeGetTypes [
    #1
    #2
    #3
    #4
   ]}
  n3 = Constant3Vector() {desc=atmos color, constant=<rgba 0.1 0.3 1 0>}
  n4 = VertexNormalWS()
  n5 = Multiply(A=Multiply(A=Fresnel(Normal=@n4) {baseReflectFraction=0, exponent=3}) {desc=atmos strength}, B=@n1.B)
  n6 = Constant3Vector() {desc=cloud color, constant=<rgba 0.8 0.8 0.82 0>}
  n7 = TextureCoordinate() {coordinateIndex=1}
  n8 = Panner(
   Coordinate=Multiply(A=@n7, B=ScalarParameter() {defaultValue=1, group=Eternity, parameterName=CloudTextureScale})
   Speed=VectorParameter() {group=Eternity, parameterName=CloudTexturePannerSpeed, defaultValue=<rgba 0.005 0 0 1>}.RGB
  ) {speedX=0.005}
  n9 = TextureSample(UVs=@n8) {samplerType=SAMPLERTYPE_LinearColor, texture=/Game/VisualThreatManager/StormVisuals/LargeClouds/T_Swirl_Flowmap}
  n10 = Multiply(A=Time()) {constB=0.03}
  n11 = Multiply(A=Subtract(A=ComponentMask(None=@n9.RGB) {g=true, r=true}) {constB=0.5}, B=Sine(None=Frac(None=@n10)))
  n12 = TextureSampleParameter2D(UVs=Add(A=@n8, B=Multiply(A=@n11) {constB=0.3})) {
   group=Eternity
   parameterName=CloudTexture
   samplerType=SAMPLERTYPE_Masks
   texture=/BRCosmetics/Accessories/FORT_Backpacks/Backpack_F_MED_Eternity/Textures/T_F_MED_Eternity_Backpack_Clouds
  }
  n13 = Add(
   A=Multiply(A=@n12.R, B=ScalarParameter() {defaultValue=0.5, group=Eternity, parameterName=CloudRedChannelMult}) {constB=0.5}
   B=Multiply(A=@n12.B, B=ScalarParameter() {defaultValue=1, group=Eternity, parameterName=CloudBlueChannelMult})
  )
  n14 = ScalarParameter() {defaultValue=4, group=Eternity, parameterName=CloudTexturePow}
  n15 = TextureSampleParameter2D(UVs=Panner(Coordinate=@n7) {speedY=0.003}) {group=Eternity, parameterName=CloudTexturem, samplerType=SAMPLERTYPE_Masks, texture=/Game/Global/Textures/Default/Noise/NoisePattern}
  n16 = MaterialFunctionCall(Float2=@n7) {materialFunction=&BreakOutFloat2Components}
  n17 = Multiply(
   A=Multiply(A=Power(Base=@n13, Exp=@n14) {constExponent=4}, B=Power(Base=@n15.G) {constExponent=1.5})
   B=SphereMask(A=Frac(None=@n16.G), B=Constant() {r=0.5}) {attenuationRadius=0.4, hardnessPercent=0}
  )
  n18 = Saturate(None=Multiply(A=Saturate(None=@n17), B=@n1.B))
  n19 = Multiply(A=@n2.Specular) {constB=0.2}
  n20 = Multiply(A=@n6, B=@n18)
  n21 = SetMaterialAttributes!(
   MaterialAttributes=@n2.MaterialAttributes
   BaseColor=LinearInterpolate(A=LinearInterpolate(A=@n2.BaseColor, B=@n3, Alpha=@n5), B=@n6, Alpha=@n18)
   Normal=LinearInterpolate(A=@n2.Normal, B=Constant3Vector() {desc=flat normal, constant=<rgba 0.5 0.5 1 0>}, Alpha=@n18) {constAlpha=1}
   Specular=Multiply(A=LinearInterpolate(A=@n19, B=@n2.Specular, Alpha=@n1.B)) {constB=0.5}
   EmissiveColor=Add(A=@n2.EmissiveColor, B=Multiply(A=@n20) {constB=0.1})
  ) {attributeSetTypes [
    #1
    #2
    #3
    #4
   ]}
  FunctionOutput(None=MaterialFunctionCall!(Alpha=@n1.G, isVehicle, Gmap, UVOffset, Local Normal, Local Position, EmissiveColorOverride, Material=@n21) {materialFunction=/Game/Weapons/WeaponSkins/MaterialFunction/PanningEmissive}.Result)
 ]}
