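# Default Lit material for the Eternity backpack cosmetic, built on material attributes for skeletal meshes and mesh deformers.
# The shared character PreFX function feeds the backpack-specific function MF_F_MED_Eternity_Backpack, whose result goes into the shared PostFX function.
# The PreFX AO output also bypasses the backpack function to drive PostFX's AO, Distance Blend and Fresnel inputs.
# The graph itself holds only these function calls, with the look delegated to them.
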Material /BRCosmetics/Accessories/FORT_Backpacks/Backpack_F_MED_Eternity/Materials/MAT_F_MED_Eternity_Backpack
graph {
 properties {
  bUseMaterialAttributes=true
  bUsedWithMeshDeformer=true
  bUsedWithSkeletalMesh=true
  pixelDepthOffsetMode=PDOM_Legacy
  usedShadingModels=Default Lit
 }
 nodes [
  n1 = MaterialFunctionCall() {materialFunction=&BaseFN_CharacterFunction_PreFX}
  MaterialFunctionCall(
   AO=@n1.AO
   Distance Blend=@n1.AO
   Fresnel=@n1.AO
   Material=MaterialFunctionCall!(In=@n1.Result) {materialFunction=&/MF_F_MED_Eternity_Backpack}.Result
  ) {materialFunction=&BaseFN_CharacterFunction_PostFX}
 ]
}
