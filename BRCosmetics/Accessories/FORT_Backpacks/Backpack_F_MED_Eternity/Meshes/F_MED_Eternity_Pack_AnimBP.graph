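# AnimBlueprint for the Eternity backpack that spins its center joint and makes eleven surrounding floater bones bob, spin and react to motion.
# EventGraph derives sine offsets and rotators from real time; additive Modify Bone nodes apply them to the input pose, cached as TransformCache.
# Two AnimDynamics spring chains, relative to the pack bone, read that cache.
# bIsOnGround blends between them: uniform springs when grounded, per-bone damped, limit-tuned springs otherwise.
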AnimBlueprint /BRCosmetics/Accessories/FORT_Backpacks/Backpack_F_MED_Eternity/Meshes/F_MED_Eternity_Pack_AnimBP
target_skeleton=&/F_MED_Eternity_Pack_Skeleton
graphs [
 AnimGraph {nodes [
   n1 = Anim Dynamics - Bone: dyn_loat_jnt_1(ComponentPose=Local To Component(LocalPose=Use cached pose 'TransformCache'()), Alpha=1, ExternalForce="0,0,0") {
    angularSpringConstant=210
    bAngularSpring=true
    bGravityOverrideInSimSpace=true
    bLinearSpring=true
    linearSpringConstant=120
    simulationSpace=BoneRelative
    alphaBoolBlend {blendOption=Linear}
    boundBone {boneName=dyn_loat_jnt_1}
    componentAppliedLinearAccClamp=<6000 6000 6000>
    componentLinearAccScale=<0.05 0.05 0.05>
    componentLinearVelScale=<0.9 0.9 0.9>
    physicsBodyDefinitions [{
      sphereCollisionRadius=0
      boundBone {boneName=dyn_loat_jnt_1}
      boxExtents=<1 1 1>
      constraintSetup {angularLimitsMax=<0 0 0>, angularLimitsMin=<0 0 0>, angularTarget=<0 0 0>}
      localJointOffset=<0 0 0>
     }]
    relativeSpaceBone {boneName=pack}
   }
   n2 = Anim Dynamics - Bone: dyn_loat_jnt_2(ComponentPose=@n1, Alpha=1, ExternalForce="0,0,0") {
    angularSpringConstant=210
    bAngularSpring=true
    bGravityOverrideInSimSpace=true
    bLinearSpring=true
    linearSpringConstant=120
    simulationSpace=BoneRelative
    alphaBoolBlend {blendOption=Linear}
    boundBone {boneName=dyn_loat_jnt_2}
    componentAppliedLinearAccClamp=<6000 6000 6000>
    componentLinearAccScale=<0.05 0.05 0.05>
    componentLinearVelScale=<0.9 0.9 0.9>
    physicsBodyDefinitions [{
      sphereCollisionRadius=0
      boundBone {boneName=dyn_loat_jnt_2}
      boxExtents=<1 1 1>
      constraintSetup {angularLimitsMax=<0 0 0>, angularLimitsMin=<0 0 0>, angularTarget=<0 0 0>}
      localJointOffset=<0 0 0>
     }]
    relativeSpaceBone {boneName=pack}
   }
   n3 = Anim Dynamics - Bone: dyn_loat_jnt_3(ComponentPose=@n2, Alpha=1, ExternalForce="0,0,0") {
    angularSpringConstant=210
    bAngularSpring=true
    bGravityOverrideInSimSpace=true
    bLinearSpring=true
    linearSpringConstant=120
    simulationSpace=BoneRelative
    alphaBoolBlend {blendOption=Linear}
    boundBone {boneName=dyn_loat_jnt_3}
    componentAppliedLinearAccClamp=<6000 6000 6000>
    componentLinearAccScale=<0.05 0.05 0.05>
    componentLinearVelScale=<0.9 0.9 0.9>
    physicsBodyDefinitions [{
      sphereCollisionRadius=0
      boundBone {boneName=dyn_loat_jnt_3}
      boxExtents=<1 1 1>
      constraintSetup {angularLimitsMax=<0 0 0>, angularLimitsMin=<0 0 0>, angularTarget=<0 0 0>}
      localJointOffset=<0 0 0>
     }]
    relativeSpaceBone {boneName=pack}
   }
   n4 = Anim Dynamics - Bone: dyn_loat_jnt_4(ComponentPose=@n3, Alpha=1, ExternalForce="0,0,0") {
    angularSpringConstant=210
    bAngularSpring=true
    bGravityOverrideInSimSpace=true
    bLinearSpring=true
    linearSpringConstant=120
    simulationSpace=BoneRelative
    alphaBoolBlend {blendOption=Linear}
    boundBone {boneName=dyn_loat_jnt_4}
    componentAppliedLinearAccClamp=<6000 6000 6000>
    componentLinearAccScale=<0.05 0.05 0.05>
    componentLinearVelScale=<0.9 0.9 0.9>
    physicsBodyDefinitions [{
      sphereCollisionRadius=0
      boundBone {boneName=dyn_loat_jnt_4}
      boxExtents=<1 1 1>
      constraintSetup {angularLimitsMax=<0 0 0>, angularLimitsMin=<0 0 0>, angularTarget=<0 0 0>}
      localJointOffset=<0 0 0>
     }]
    relativeSpaceBone {boneName=pack}
   }
   n5 = Anim Dynamics - Bone: dyn_loat_jnt_5(ComponentPose=@n4, Alpha=1, ExternalForce="0,0,0") {
    angularSpringConstant=210
    bAngularSpring=true
    bGravityOverrideInSimSpace=true
    bLinearSpring=true
    linearSpringConstant=120
    simulationSpace=BoneRelative
    alphaBoolBlend {blendOption=Linear}
    boundBone {boneName=dyn_loat_jnt_5}
    componentAppliedLinearAccClamp=<6000 6000 6000>
    componentLinearAccScale=<0.05 0.05 0.05>
    componentLinearVelScale=<0.9 0.9 0.9>
    physicsBodyDefinitions [{
      sphereCollisionRadius=0
      boundBone {boneName=dyn_loat_jnt_5}
      boxExtents=<1 1 1>
      constraintSetup {angularLimitsMax=<0 0 0>, angularLimitsMin=<0 0 0>, angularTarget=<0 0 0>}
      localJointOffset=<0 0 0>
     }]
    relativeSpaceBone {boneName=pack}
   }
   n6 = Anim Dynamics - Bone: dyn_loat_jnt_6(ComponentPose=@n5, Alpha=1, ExternalForce="0,0,0") {
    angularSpringConstant=210
    bAngularSpring=true
    bGravityOverrideInSimSpace=true
    bLinearSpring=true
    linearSpringConstant=120
    simulationSpace=BoneRelative
    alphaBoolBlend {blendOption=Linear}
    boundBone {boneName=dyn_loat_jnt_6}
    componentAppliedLinearAccClamp=<6000 6000 6000>
    componentLinearAccScale=<0.05 0.05 0.05>
    componentLinearVelScale=<0.9 0.9 0.9>
    physicsBodyDefinitions [{
      sphereCollisionRadius=0
      boundBone {boneName=dyn_loat_jnt_6}
      boxExtents=<1 1 1>
      constraintSetup {angularLimitsMax=<0 0 0>, angularLimitsMin=<0 0 0>, angularTarget=<0 0 0>}
      localJointOffset=<0 0 0>
     }]
    relativeSpaceBone {boneName=pack}
   }
   n7 = Anim Dynamics - Bone: dyn_loat_jnt_7(ComponentPose=@n6, Alpha=1, ExternalForce="0,0,0") {
    angularSpringConstant=210
    bAngularSpring=true
    bGravityOverrideInSimSpace=true
    bLinearSpring=true
    linearSpringConstant=120
    simulationSpace=BoneRelative
    alphaBoolBlend {blendOption=Linear}
    boundBone {boneName=dyn_loat_jnt_7}
    componentAppliedLinearAccClamp=<6000 6000 6000>
    componentLinearAccScale=<0.05 0.05 0.05>
    componentLinearVelScale=<0.9 0.9 0.9>
    physicsBodyDefinitions [{
      sphereCollisionRadius=0
      boundBone {boneName=dyn_loat_jnt_7}
      boxExtents=<1 1 1>
      constraintSetup {angularLimitsMax=<0 0 0>, angularLimitsMin=<0 0 0>, angularTarget=<0 0 0>}
      localJointOffset=<0 0 0>
     }]
    relativeSpaceBone {boneName=pack}
   }
   n8 = Anim Dynamics - Bone: dyn_loat_jnt_8(ComponentPose=@n7, Alpha=1, ExternalForce="0,0,0") {
    angularSpringConstant=210
    bAngularSpring=true
    bGravityOverrideInSimSpace=true
    bLinearSpring=true
    linearSpringConstant=120
    simulationSpace=BoneRelative
    alphaBoolBlend {blendOption=Linear}
    boundBone {boneName=dyn_loat_jnt_8}
    componentAppliedLinearAccClamp=<6000 6000 6000>
    componentLinearAccScale=<0.05 0.05 0.05>
    componentLinearVelScale=<0.9 0.9 0.9>
    physicsBodyDefinitions [{
      sphereCollisionRadius=0
      boundBone {boneName=dyn_loat_jnt_8}
      boxExtents=<1 1 1>
      constraintSetup {angularLimitsMax=<0 0 0>, angularLimitsMin=<0 0 0>, angularTarget=<0 0 0>}
      localJointOffset=<0 0 0>
     }]
    relativeSpaceBone {boneName=pack}
   }
   n9 = Anim Dynamics - Bone: dyn_loat_jnt_9(ComponentPose=@n8, Alpha=1, ExternalForce="0,0,0") {
    angularSpringConstant=210
    bAngularSpring=true
    bGravityOverrideInSimSpace=true
    bLinearSpring=true
    linearSpringConstant=120
    simulationSpace=BoneRelative
    alphaBoolBlend {blendOption=Linear}
    boundBone {boneName=dyn_loat_jnt_9}
    componentAppliedLinearAccClamp=<6000 6000 6000>
    componentLinearAccScale=<0.05 0.05 0.05>
    componentLinearVelScale=<0.9 0.9 0.9>
    physicsBodyDefinitions [{
      sphereCollisionRadius=0
      boundBone {boneName=dyn_loat_jnt_9}
      boxExtents=<1 1 1>
      constraintSetup {angularLimitsMax=<0 0 0>, angularLimitsMin=<0 0 0>, angularTarget=<0 0 0>}
      localJointOffset=<0 0 0>
     }]
    relativeSpaceBone {boneName=pack}
   }
   n10 = Anim Dynamics - Bone: dyn_loat_jnt_10(ComponentPose=@n9, Alpha=1, ExternalForce="0,0,0") {
    angularSpringConstant=210
    bAngularSpring=true
    bGravityOverrideInSimSpace=true
    bLinearSpring=true
    linearSpringConstant=120
    simulationSpace=BoneRelative
    alphaBoolBlend {blendOption=Linear}
    boundBone {boneName=dyn_loat_jnt_10}
    componentAppliedLinearAccClamp=<6000 6000 6000>
    componentLinearAccScale=<0.05 0.05 0.05>
    componentLinearVelScale=<0.9 0.9 0.9>
    physicsBodyDefinitions [{
      sphereCollisionRadius=0
      boundBone {boneName=dyn_loat_jnt_10}
      boxExtents=<1 1 1>
      constraintSetup {angularLimitsMax=<0 0 0>, angularLimitsMin=<0 0 0>, angularTarget=<0 0 0>}
      localJointOffset=<0 0 0>
     }]
    relativeSpaceBone {boneName=pack}
   }
   n11 = Anim Dynamics - Bone: dyn_loat_jnt_11(ComponentPose=@n10, Alpha=1, ExternalForce="0,0,0") {
    angularSpringConstant=210
    bAngularSpring=true
    bGravityOverrideInSimSpace=true
    bLinearSpring=true
    linearSpringConstant=120
    simulationSpace=BoneRelative
    alphaBoolBlend {blendOption=Linear}
    boundBone {boneName=dyn_loat_jnt_11}
    componentAppliedLinearAccClamp=<6000 6000 6000>
    componentLinearAccScale=<0.05 0.05 0.05>
    componentLinearVelScale=<0.9 0.9 0.9>
    physicsBodyDefinitions [{
      sphereCollisionRadius=0
      boundBone {boneName=dyn_loat_jnt_11}
      boxExtents=<1 1 1>
      constraintSetup {angularLimitsMax=<0 0 0>, angularLimitsMin=<0 0 0>, angularTarget=<0 0 0>}
      localJointOffset=<0 0 0>
     }]
    relativeSpaceBone {boneName=pack}
   }
   n12 = Anim Dynamics - Bone: dyn_loat_jnt_1(ComponentPose=Local To Component(LocalPose=Use cached pose 'TransformCache'()), Alpha=0.7, ExternalForce="0,0,0") {
    angularDampingOverride=0.9
    angularSpringConstant=50
    bAngularSpring=true
    bGravityOverrideInSimSpace=true
    bLinearSpring=true
    bOverrideAngularDamping=true
    bOverrideLinearDamping=true
    gravityScale=0
    linearDampingOverride=0.9999
    linearSpringConstant=50
    simulationSpace=BoneRelative
    alphaBoolBlend {blendOption=Linear}
    boundBone {boneName=dyn_loat_jnt_1}
    componentAppliedLinearAccClamp=<6000 6000 6000>
    componentLinearAccScale=<0.05 0.05 0.02>
    componentLinearVelScale=<2 2 0.02>
    gravityOverride {z=-980}
    physicsBodyDefinitions [{
      sphereCollisionRadius=0
      boundBone {boneName=dyn_loat_jnt_1}
      boxExtents=<50 50 50>
      constraintSetup {angularLimitsMax=<0 0 0>, angularLimitsMin=<0 0 0>, angularTarget=<0 0 0>, linearAxesMax=<2 3 3>, linearAxesMin=<-2 -1 -3>}
      localJointOffset=<0 0 0>
     }]
    relativeSpaceBone {boneName=pack}
   }
   n13 = Anim Dynamics - Bone: dyn_loat_jnt_2(ComponentPose=@n12, Alpha=1, ExternalForce="0,0,0") {
    angularDampingOverride=0.9
    angularSpringConstant=210
    bAngularSpring=true
    bGravityOverrideInSimSpace=true
    bLinearSpring=true
    bOverrideAngularDamping=true
    bOverrideLinearDamping=true
    gravityScale=0
    linearDampingOverride=0.9999
    linearSpringConstant=90
    simulationSpace=BoneRelative
    alphaBoolBlend {blendOption=Linear}
    boundBone {boneName=dyn_loat_jnt_2}
    componentAppliedLinearAccClamp=<6000 6000 6000>
    componentLinearAccScale=<0.01 0.01 0.01>
    componentLinearVelScale=<0.4 0.4 0.05>
    gravityOverride {z=-980}
    physicsBodyDefinitions [{
      sphereCollisionRadius=0
      boundBone {boneName=dyn_loat_jnt_2}
      boxExtents=<50 50 50>
      constraintSetup {angularLimitsMax=<0 0 0>, angularLimitsMin=<0 0 0>, angularTarget=<0 0 0>, linearAxesMax=<2 3 2>, linearAxesMin=<-2 -1 -4>}
      localJointOffset=<-0.5 -0.5 -0.5>
     }]
    relativeSpaceBone {boneName=pack}
   }
   n14 = Anim Dynamics - Bone: dyn_loat_jnt_3(ComponentPose=@n13, Alpha=1, ExternalForce="0,0,0") {
    angularDampingOverride=0.9
    angularSpringConstant=250
    bAngularSpring=true
    bGravityOverrideInSimSpace=true
    bLinearSpring=true
    bOverrideAngularDamping=true
    bOverrideLinearDamping=true
    gravityScale=0
    linearDampingOverride=0.9999
    linearSpringConstant=150
    simulationSpace=BoneRelative
    alphaBoolBlend {blendOption=Linear}
    boundBone {boneName=dyn_loat_jnt_3}
    componentAppliedLinearAccClamp=<6000 6000 6000>
    componentLinearAccScale=<0.05 0.05 0.04>
    componentLinearVelScale=<1.3 1.3 0.7>
    gravityOverride {z=-980}
    physicsBodyDefinitions [{
      sphereCollisionRadius=0
      boundBone {boneName=dyn_loat_jnt_3}
      boxExtents=<50 50 50>
      constraintSetup {angularLimitsMax=<0 0 0>, angularLimitsMin=<0 0 0>, angularTarget=<0 0 0>, linearAxesMax=<3 3 2.4>, linearAxesMin=<-3 -1 -3>}
      localJointOffset=<-1 -1 -1>
     }]
    relativeSpaceBone {boneName=pack}
   }
   n15 = Anim Dynamics - Bone: dyn_loat_jnt_4(ComponentPose=@n14, Alpha=1, ExternalForce="0,0,0") {
    angularDampingOverride=0.9
    angularSpringConstant=210
    bAngularSpring=true
    bGravityOverrideInSimSpace=true
    bLinearSpring=true
    bOverrideAngularDamping=true
    bOverrideLinearDamping=true
    gravityScale=0
    linearDampingOverride=0.9999
    linearSpringConstant=100
    simulationSpace=BoneRelative
    alphaBoolBlend {blendOption=Linear}
    boundBone {boneName=dyn_loat_jnt_4}
    componentAppliedLinearAccClamp=<6000 6000 6000>
    componentLinearAccScale=<0.01 0.01 0.01>
    componentLinearVelScale=<0.9 0.9 0.09>
    gravityOverride {z=-980}
    physicsBodyDefinitions [{
      sphereCollisionRadius=0
      boundBone {boneName=dyn_loat_jnt_4}
      boxExtents=<50 50 50>
      constraintSetup {angularLimitsMax=<0 0 0>, angularLimitsMin=<0 0 0>, angularTarget=<0 0 0>, linearAxesMax=<2 3 1.5>, linearAxesMin=<-2 -1 -4>}
      localJointOffset=<-0.7 -0.7 -0.7>
     }]
    relativeSpaceBone {boneName=pack}
   }
   n16 = Anim Dynamics - Bone: dyn_loat_jnt_5(ComponentPose=@n15, Alpha=1, ExternalForce="0,0,0") {
    angularDampingOverride=0.9
    angularSpringConstant=160
    bAngularSpring=true
    bGravityOverrideInSimSpace=true
    bLinearSpring=true
    bOverrideAngularDamping=true
    bOverrideLinearDamping=true
    gravityScale=0
    linearDampingOverride=0.9999
    linearSpringConstant=180
    simulationSpace=BoneRelative
    alphaBoolBlend {blendOption=Linear}
    boundBone {boneName=dyn_loat_jnt_5}
    componentAppliedLinearAccClamp=<6000 6000 6000>
    componentLinearAccScale=<0.05 0.05 0.05>
    componentLinearVelScale=<0.2 0.2 0.1>
    gravityOverride {z=-980}
    physicsBodyDefinitions [{
      sphereCollisionRadius=0
      boundBone {boneName=dyn_loat_jnt_5}
      boxExtents=<50 50 50>
      constraintSetup {angularLimitsMax=<0 0 0>, angularLimitsMin=<0 0 0>, angularTarget=<0 0 0>, linearAxesMax=<2 3 3.3>, linearAxesMin=<-2 0 -4>}
      localJointOffset=<-1 -1 -1>
     }]
    relativeSpaceBone {boneName=pack}
   }
   n17 = Anim Dynamics - Bone: dyn_loat_jnt_6(ComponentPose=@n16, Alpha=1, ExternalForce="0,0,0") {
    angularDampingOverride=0.9
    angularSpringConstant=210
    bAngularSpring=true
    bGravityOverrideInSimSpace=true
    bLinearSpring=true
    bOverrideAngularDamping=true
    bOverrideLinearDamping=true
    gravityScale=0
    linearDampingOverride=0.9999
    linearSpringConstant=90
    simulationSpace=BoneRelative
    alphaBoolBlend {blendOption=Linear}
    boundBone {boneName=dyn_loat_jnt_6}
    componentAppliedLinearAccClamp=<6000 6000 6000>
    componentLinearAccScale=<0.01 0.01 0.01>
    componentLinearVelScale=<1 1 0.9>
    gravityOverride {z=-980}
    physicsBodyDefinitions [{
      sphereCollisionRadius=0
      boundBone {boneName=dyn_loat_jnt_6}
      boxExtents=<50 50 50>
      constraintSetup {angularLimitsMax=<0 0 0>, angularLimitsMin=<0 0 0>, angularTarget=<0 0 0>, linearAxesMax=<2 3 2.2>, linearAxesMin=<-2 0 -2>}
      localJointOffset=<-0.5 -0.5 -0.5>
     }]
    relativeSpaceBone {boneName=pack}
   }
   n18 = Anim Dynamics - Bone: dyn_loat_jnt_7(ComponentPose=@n17, Alpha=1, ExternalForce="0,0,0") {
    angularDampingOverride=0.9
    angularSpringConstant=210
    bAngularSpring=true
    bGravityOverrideInSimSpace=true
    bLinearSpring=true
    bOverrideAngularDamping=true
    bOverrideLinearDamping=true
    gravityScale=0
    linearDampingOverride=0.9999
    linearSpringConstant=120
    simulationSpace=BoneRelative
    alphaBoolBlend {blendOption=Linear}
    boundBone {boneName=dyn_loat_jnt_7}
    componentAppliedLinearAccClamp=<6000 6000 6000>
    componentLinearAccScale=<0.05 0.05 0.05>
    componentLinearVelScale=<0.6 0.6 0.6>
    gravityOverride {z=-980}
    physicsBodyDefinitions [{
      sphereCollisionRadius=0
      boundBone {boneName=dyn_loat_jnt_7}
      boxExtents=<50 50 50>
      constraintSetup {angularLimitsMax=<0 0 0>, angularLimitsMin=<0 0 0>, angularTarget=<0 0 0>, linearAxesMax=<2 3 3>, linearAxesMin=<-2 0 -4>}
      localJointOffset=<-0.8 -0.8 -0.8>
     }]
    relativeSpaceBone {boneName=pack}
   }
   n19 = Anim Dynamics - Bone: dyn_loat_jnt_8(ComponentPose=@n18, Alpha=1, ExternalForce="0,0,0") {
    angularDampingOverride=0.9
    angularSpringConstant=210
    bAngularSpring=true
    bGravityOverrideInSimSpace=true
    bLinearSpring=true
    bOverrideAngularDamping=true
    bOverrideLinearDamping=true
    gravityScale=0
    linearDampingOverride=0.9999
    linearSpringConstant=120
    simulationSpace=BoneRelative
    alphaBoolBlend {blendOption=Linear}
    boundBone {boneName=dyn_loat_jnt_8}
    componentAppliedLinearAccClamp=<6000 6000 6000>
    componentLinearAccScale=<0.01 0.01 0.01>
    componentLinearVelScale=<0.05 0.05 0.03>
    gravityOverride {z=-980}
    physicsBodyDefinitions [{
      sphereCollisionRadius=0
      boundBone {boneName=dyn_loat_jnt_8}
      boxExtents=<50 50 50>
      constraintSetup {angularLimitsMax=<0 0 0>, angularLimitsMin=<0 0 0>, angularTarget=<0 0 0>, linearAxesMax=<2 3 2>, linearAxesMin=<-2 0 -2>}
      localJointOffset=<-0.4 -0.4 -0.4>
     }]
    relativeSpaceBone {boneName=pack}
   }
   n20 = Anim Dynamics - Bone: dyn_loat_jnt_9(ComponentPose=@n19, Alpha=1, ExternalForce="0,0,0") {
    angularDampingOverride=0.9
    angularSpringConstant=210
    bAngularSpring=true
    bGravityOverrideInSimSpace=true
    bLinearSpring=true
    bOverrideAngularDamping=true
    bOverrideLinearDamping=true
    gravityScale=0
    linearDampingOverride=0.9999
    linearSpringConstant=120
    simulationSpace=BoneRelative
    alphaBoolBlend {blendOption=Linear}
    boundBone {boneName=dyn_loat_jnt_9}
    componentAppliedLinearAccClamp=<6000 6000 6000>
    componentLinearAccScale=<0.05 0.05 0.05>
    componentLinearVelScale=<1 1 0.2>
    gravityOverride {z=-980}
    physicsBodyDefinitions [{
      sphereCollisionRadius=0
      boundBone {boneName=dyn_loat_jnt_9}
      boxExtents=<50 50 50>
      constraintSetup {angularLimitsMax=<0 0 0>, angularLimitsMin=<0 0 0>, angularTarget=<0 0 0>, linearAxesMax=<2 3 2.2>, linearAxesMin=<-2 1 -4>}
      localJointOffset=<-1.1 -1.1 -1.1>
     }]
    relativeSpaceBone {boneName=pack}
   }
   n21 = Anim Dynamics - Bone: dyn_loat_jnt_10(ComponentPose=@n20, Alpha=1, ExternalForce="0,0,0") {
    angularDampingOverride=0.9
    angularSpringConstant=210
    bAngularSpring=true
    bGravityOverrideInSimSpace=true
    bLinearSpring=true
    bOverrideAngularDamping=true
    bOverrideLinearDamping=true
    gravityScale=0
    linearDampingOverride=0.9999
    linearSpringConstant=100
    simulationSpace=BoneRelative
    alphaBoolBlend {blendOption=Linear}
    boundBone {boneName=dyn_loat_jnt_10}
    componentAppliedLinearAccClamp=<6000 6000 6000>
    componentLinearAccScale=<0.01 0.01 0.01>
    componentLinearVelScale=<0.4 0.4 0.4>
    gravityOverride {z=-980}
    physicsBodyDefinitions [{
      sphereCollisionRadius=0
      boundBone {boneName=dyn_loat_jnt_10}
      boxExtents=<50 50 50>
      constraintSetup {angularLimitsMax=<0 0 0>, angularLimitsMin=<0 0 0>, angularTarget=<0 0 0>, linearAxesMax=<4 3 2.5>, linearAxesMin=<-4 0 -2>}
      localJointOffset=<-0.5 -0.5 -0.5>
     }]
    relativeSpaceBone {boneName=pack}
   }
   n22 = Anim Dynamics - Bone: dyn_loat_jnt_11(ComponentPose=@n21, Alpha=1, ExternalForce="0,0,0") {
    angularDampingOverride=0.9
    angularSpringConstant=210
    bAngularSpring=true
    bGravityOverrideInSimSpace=true
    bLinearSpring=true
    bOverrideAngularDamping=true
    bOverrideLinearDamping=true
    gravityScale=0
    linearDampingOverride=0.9999
    linearSpringConstant=100
    simulationSpace=BoneRelative
    alphaBoolBlend {blendOption=Linear}
    boundBone {boneName=dyn_loat_jnt_11}
    componentAppliedLinearAccClamp=<6000 6000 6000>
    componentLinearAccScale=<0.05 0.05 0.05>
    componentLinearVelScale=<0.1 0.1 0.05>
    gravityOverride {z=-980}
    physicsBodyDefinitions [{
      sphereCollisionRadius=0
      boundBone {boneName=dyn_loat_jnt_11}
      boxExtents=<50 50 50>
      constraintSetup {angularLimitsMax=<0 0 0>, angularLimitsMin=<0 0 0>, angularTarget=<0 0 0>, linearAxesMax=<2 3 1.9>, linearAxesMin=<-2 1 -3.5>}
      localJointOffset=<-1 -1 -1>
     }]
    relativeSpaceBone {boneName=pack}
   }
   Output Pose(Result=Blend Poses by bool(
     BlendPose_0=Component To Local(ComponentPose=@n11)
     BlendPose_1=Component To Local(ComponentPose=@n22)
     BlendTime_0=0.05
     BlendTime_1=0.05
     bActiveValue=Get bIsOnGround() -> (bIsOnGround:bool)
    ) {blendType=Linear}) {layerGroup=None}
   n23 = "Transform (Modify) Bone - Bone: dyn_centered_jnt"(
    ComponentPose=Local To Component(LocalPose=Input Pose())
    Alpha=1
    Translation="0,0,0"
    Rotation=Get spinGlobe() -> (spinGlobe:Rotator)
    Scale="1,1,1"
   ) {
    rotationMode=BMM_Additive
    scaleSpace=BCS_BoneSpace
    translationMode=BMM_Additive
    translationSpace=BCS_BoneSpace
    alphaBoolBlend {blendOption=Linear}
    boneToModify {boneName=dyn_centered_jnt}
   }
   n24 = "Transform (Modify) Bone - Bone: dyn_loat_jnt_1"(ComponentPose=@n23, Alpha=1, Translation="Get FloaterA(more)"() -> ("FloaterA(more)":Vector), Rotation="-2.080,0.000,0.000", Scale="1,1,1") {
    rotationMode=BMM_Additive
    scaleSpace=BCS_BoneSpace
    translationMode=BMM_Additive
    translationSpace=BCS_BoneSpace
    alphaBoolBlend {blendOption=Linear}
    boneToModify {boneName=dyn_loat_jnt_1}
    rotation {pitch=-2.08}
   }
   n25 = "Get SpinningRotator2(lesser)"() -> ("SpinningRotator2(lesser)":Rotator)
   n26 = "Transform (Modify) Bone - Bone: dyn_loat_jnt_2"(ComponentPose=@n24, Alpha=1, Translation=Get Floater Most LeftRight() -> (Floater Most LeftRight:Vector), Rotation=@n25, Scale="1,1,1") {
    rotationMode=BMM_Additive
    rotationSpace=BCS_BoneSpace
    translationMode=BMM_Additive
    translationSpace=BCS_BoneSpace
    alphaBoolBlend {blendOption=Linear}
    boneToModify {boneName=dyn_loat_jnt_2}
   }
   n27 = "Get SpinningRotator1(more)"() -> ("SpinningRotator1(more)":Rotator)
   n28 = Get Floater Less LeftRight() -> (Floater Less LeftRight:Vector)
   n29 = "Transform (Modify) Bone - Bone: dyn_loat_jnt_3"(ComponentPose=@n26, Alpha=1, Translation=@n28, Rotation=@n27, Scale="1,1,1") {
    rotationMode=BMM_Additive
    rotationSpace=BCS_BoneSpace
    translationMode=BMM_Additive
    translationSpace=BCS_BoneSpace
    alphaBoolBlend {blendOption=Linear}
    boneToModify {boneName=dyn_loat_jnt_3}
   }
   n30 = "Get FloaterC(most)"() -> ("FloaterC(most)":Vector)
   n31 = "Transform (Modify) Bone - Bone: dyn_loat_jnt_4"(ComponentPose=@n29, Alpha=1, Translation=@n30, Rotation=@n27, Scale="1,1,1") {
    rotationMode=BMM_Additive
    rotationSpace=BCS_BoneSpace
    translationMode=BMM_Additive
    translationSpace=BCS_BoneSpace
    alphaBoolBlend {blendOption=Linear}
    boneToModify {boneName=dyn_loat_jnt_4}
   }
   n32 = "Get FloaterB(lesser)"() -> ("FloaterB(lesser)":Vector)
   n33 = "Transform (Modify) Bone - Bone: dyn_loat_jnt_5"(ComponentPose=@n31, Alpha=1, Translation=@n32, Rotation=@n25, Scale="1,1,1") {
    rotationMode=BMM_Additive
    rotationSpace=BCS_BoneSpace
    translationMode=BMM_Additive
    translationSpace=BCS_BoneSpace
    alphaBoolBlend {blendOption=Linear}
    boneToModify {boneName=dyn_loat_jnt_5}
   }
   n34 = "Transform (Modify) Bone - Bone: dyn_loat_jnt_6"(ComponentPose=@n33, Alpha=1, Translation=@n28, Rotation=@n27, Scale="1,1,1") {
    rotationMode=BMM_Additive
    rotationSpace=BCS_BoneSpace
    translationMode=BMM_Additive
    translationSpace=BCS_BoneSpace
    alphaBoolBlend {blendOption=Linear}
    boneToModify {boneName=dyn_loat_jnt_6}
   }
   n35 = "Transform (Modify) Bone - Bone: dyn_loat_jnt_7"(ComponentPose=@n34, Alpha=1, Translation=@n32, Rotation="0,0,0", Scale="1,1,1") {
    rotationMode=BMM_Additive
    rotationSpace=BCS_BoneSpace
    translationMode=BMM_Additive
    translationSpace=BCS_BoneSpace
    alphaBoolBlend {blendOption=Linear}
    boneToModify {boneName=dyn_loat_jnt_7}
   }
   n36 = "Transform (Modify) Bone - Bone: dyn_loat_jnt_8"(ComponentPose=@n35, Alpha=1, Translation=Get Floater More LeftRight() -> (Floater More LeftRight:Vector), Rotation=@n27, Scale="1,1,1") {
    rotationMode=BMM_Additive
    rotationSpace=BCS_BoneSpace
    translationMode=BMM_Additive
    translationSpace=BCS_BoneSpace
    alphaBoolBlend {blendOption=Linear}
    boneToModify {boneName=dyn_loat_jnt_8}
   }
   n37 = "Transform (Modify) Bone - Bone: dyn_loat_jnt_9"(ComponentPose=@n36, Alpha=1, Translation=@n32, Rotation=@n25, Scale="1,1,1") {
    rotationMode=BMM_Additive
    rotationSpace=BCS_BoneSpace
    translationMode=BMM_Additive
    translationSpace=BCS_BoneSpace
    alphaBoolBlend {blendOption=Linear}
    boneToModify {boneName=dyn_loat_jnt_9}
   }
   n38 = "Transform (Modify) Bone - Bone: dyn_loat_jnt_10"(ComponentPose=@n37, Alpha=1, Translation=@n30, Rotation=@n27, Scale="1,1,1") {
    rotationMode=BMM_Additive
    rotationSpace=BCS_BoneSpace
    translationMode=BMM_Additive
    translationSpace=BCS_BoneSpace
    alphaBoolBlend {blendOption=Linear}
    boneToModify {boneName=dyn_loat_jnt_10}
   }
   n39 = "Transform (Modify) Bone - Bone: dyn_loat_jnt_11"(ComponentPose=@n38, Alpha=1, Translation=@n32, Rotation=@n27, Scale="1,1,1") {
    rotationMode=BMM_Additive
    rotationSpace=BCS_BoneSpace
    translationMode=BMM_Additive
    translationSpace=BCS_BoneSpace
    alphaBoolBlend {blendOption=Linear}
    boneToModify {boneName=dyn_loat_jnt_11}
   }
   Save cached pose 'TransformCache'(Pose=Component To Local(ComponentPose=@n39)) {cacheName=TransformCache}
  ]}
 EventGraph {nodes [
   Event BlueprintUpdateAnimation()
   > n1 = Sequence!(execute) -> (then_0, then_1, then_2:Exec, then_3:Exec, then_4:Exec, then_5:Exec, then_6:Exec, then_7:Exec, then_8:Exec)
   n2 = float / float(A=GetRealTimeSeconds(), B=-100) {is=title K2Node_CallFunction}
   > "Set SpinningRotator1(more)"("SpinningRotator1(more)":Rotator=MakeRotator(Roll=float * float(A=@n2, B=10000))) -> (Output_Get:Rotator)
   n3 = float * float(A=GetRealTimeSeconds(), B=2)
   "Set FloaterA(more)"(execute=@n1.then_1, "FloaterA(more)":Vector=MakeVector(Y=0, Z="Sin (Radians)"(A=@n3))) -> (Output_Get:Vector)
   n4 = float * float(A=GetRealTimeSeconds(), B=1.25)
   "Set FloaterB(lesser)"(execute=@n1.then_2, "FloaterB(lesser)":Vector=MakeVector(Y=0, Z="Sin (Radians)"(A=@n4))) -> (Output_Get:Vector)
   n5 = float / float(A=GetRealTimeSeconds(), B=-500)
   "Set SpinningRotator2(lesser)"(execute=@n1.then_3, "SpinningRotator2(lesser)":Rotator=MakeRotator(Roll=float * float(A=@n5, B=30000))) -> (Output_Get:Rotator)
   n6 = float * float(A=GetRealTimeSeconds(), B=2.25)
   "Set FloaterC(most)"(execute=@n1.then_4, "FloaterC(most)":Vector=MakeVector(Y=0, Z="Sin (Radians)"(A=@n6))) -> (Output_Get:Vector)
   n7 = "Sin (Radians)"(A=float * float(A=GetRealTimeSeconds(), B=2))
   Set Floater More LeftRight(execute=@n1.then_5, Floater More LeftRight:Vector=MakeVector(X=@n7, Y=0, Z=@n7)) -> (Output_Get:Vector)
   n8 = "Sin (Radians)"(A=float * float(A=GetRealTimeSeconds(), B=1.25))
   Set Floater Less LeftRight(execute=@n1.then_6, Floater Less LeftRight:Vector=MakeVector(X=@n8, Y=0, Z=@n8)) -> (Output_Get:Vector)
   n9 = "Sin (Radians)"(A=float * float(A=GetRealTimeSeconds(), B=2.15))
   Set Floater Most LeftRight(execute=@n1.then_7, Floater Most LeftRight:Vector=MakeVector(X=@n9, Y=0, Z=@n9)) -> (Output_Get:Vector)
   n10 = float / float(A=GetRealTimeSeconds(), B=50)
   Set spinGlobe(execute=@n1.then_8, spinGlobe:Rotator=MakeRotator(Yaw=float * float(A=@n10, B=500))) -> (Output_Get:Rotator)
   TryGetPawnOwner()
  ]}
]
variables [
 Floater Less LeftRight {type=Vector}
 Floater More LeftRight {type=Vector}
 Floater Most LeftRight {type=Vector}
 "FloaterA(more)" {type=Vector}
 "FloaterB(lesser)" {type=Vector}
 "FloaterC(most)" {type=Vector}
 spinGlobe {type=Rotator}
 "SpinningRotator1(more)" {type=Rotator}
 "SpinningRotator2(lesser)" {type=Rotator}
]
